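# Physics asset for the VagueHex backpack cosmetic, giving its attached character figure simulated secondary motion.
# Capsule bodies cover the pack, torso, head, spine, two-segment arms and legs, and a four-segment ear chain.
# The torso hangs from the pack and carries head, arms and spine, with the ear chain off the head and legs off the spine.
# Tight swing and twist limits restrict each joint, with zero twist on head and ear.
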PhysicsAsset /BRCosmetics/Accessories/FORT_Backpacks/Backpack_F_MED_VagueHex/Meshes/F_MED_VagueHex_Pack_Physics
bodies {
 dyn_arm_l_01 [{shapeType=Capsule, length=6, radius=1.2, shapeName=dyn_arm_l_01_capsule, center=<4 0 0>, rotation=<rot 90 0 0>}]
 dyn_arm_l_02 [{shapeType=Capsule, length=4, radius=1.5, shapeName=dyn_arm_l_02_capsule, center=<2 0 0>, rotation=<rot 90 0 0>}]
 dyn_arm_r_01 [{shapeType=Capsule, length=6, radius=1.2, shapeName=dyn_arm_r_01_capsule, center=<4 0 0>, rotation=<rot 90 0 0>}]
 dyn_arm_r_02 [{shapeType=Capsule, length=4, radius=1.5, shapeName=dyn_arm_r_02_capsule, center=<2 0 0>, rotation=<rot 90 0 0>}]
 dyn_ear_01 [{shapeType=Capsule, length=2, radius=1, shapeName=dyn_ear_01_capsule, center=<2 0 0>, rotation=<rot 90 0 0>}]
 dyn_ear_02 [{shapeType=Capsule, length=4, radius=1, shapeName=dyn_ear_02_capsule, center=<4 0 0>, rotation=<rot 90 0 0>}]
 dyn_ear_03 [{shapeType=Capsule, length=3.5, radius=1, shapeName=dyn_ear_03_capsule, center=<3 0 0>, rotation=<rot 90 0 0>}]
 dyn_ear_04 [{shapeType=Capsule, length=2, radius=1, shapeName=dyn_ear_04_capsule, center=<2 0 0>, rotation=<rot 90 0 0>}]
 dyn_head [{shapeType=Capsule, length=3.3674, radius=8.5555, shapeName=dyn_head_capsule, center=<3 0 0>, rotation=<rot 0 0 90>}]
 dyn_leg_l_01 [{shapeType=Capsule, length=5, radius=1.6, shapeName=dyn_leg_l_01_capsule, center=<4 0 0>, rotation=<rot 90 0 0>}]
 dyn_leg_l_02 [{shapeType=Capsule, length=6, radius=2, shapeName=dyn_leg_l_02_capsule, center=<6 0 0>, rotation=<rot 90 0 0>}]
 dyn_leg_r_01 [{shapeType=Capsule, length=5, radius=1.6, shapeName=dyn_leg_r_01_capsule, center=<4 0 0>, rotation=<rot 90 0 0>}]
 dyn_leg_r_02 [{shapeType=Capsule, length=6, radius=2, shapeName=dyn_leg_r_02_capsule, center=<6 0 0>, rotation=<rot 90 0 0>}]
 dyn_spine [{shapeType=Capsule, length=0, radius=4.5, shapeName=dyn_spine_capsule, center=<1 0 0>, rotation=<rot 90 0 0>}]
 dyn_torso [{shapeType=Capsule, length=0, radius=3.5, shapeName=dyn_torso_capsule, center=<2 0 0>, rotation=<rot 90 0 0>}]
 pack [{shapeType=Capsule, length=0.505, radius=0.505, shapeName=pack_capsule, rotation=<rot 0 -90 0>}]
}
constraints [
 {bone1Name=dyn_arm_l_01, bone2Name=dyn_torso, swing1LimitDegrees=20, swing2LimitDegrees=15, twistLimitDegrees=5}
 {bone1Name=dyn_arm_l_02, bone2Name=dyn_arm_l_01, swing1LimitDegrees=20, swing2LimitDegrees=15, twistLimitDegrees=5}
 {bone1Name=dyn_arm_r_01, bone2Name=dyn_torso, swing1LimitDegrees=20, swing2LimitDegrees=15, twistLimitDegrees=5}
 {bone1Name=dyn_arm_r_02, bone2Name=dyn_arm_r_01, swing1LimitDegrees=20, swing2LimitDegrees=15, twistLimitDegrees=5}
 {bone1Name=dyn_head, bone2Name=dyn_torso, swing1LimitDegrees=10, swing2LimitDegrees=10, twistLimitDegrees=0}
 {bone1Name=dyn_ear_01, bone2Name=dyn_head, swing1LimitDegrees=5, swing2LimitDegrees=5, twistLimitDegrees=0}
 {bone1Name=dyn_ear_02, bone2Name=dyn_ear_01, swing1LimitDegrees=15, swing2LimitDegrees=20, twistLimitDegrees=0}
 {bone1Name=dyn_ear_03, bone2Name=dyn_ear_02, swing1LimitDegrees=15, swing2LimitDegrees=20, twistLimitDegrees=0}
 {bone1Name=dyn_ear_04, bone2Name=dyn_ear_03, swing1LimitDegrees=15, swing2LimitDegrees=20, twistLimitDegrees=0}
 {bone1Name=dyn_spine, bone2Name=dyn_torso, swing1LimitDegrees=15, swing2LimitDegrees=25, twistLimitDegrees=5}
 {bone1Name=dyn_leg_l_01, bone2Name=dyn_spine, swing1LimitDegrees=10, swing2LimitDegrees=15, twistLimitDegrees=5}
 {bone1Name=dyn_leg_l_02, bone2Name=dyn_leg_l_01, swing1LimitDegrees=10, swing2LimitDegrees=15, twistLimitDegrees=5}
 {bone1Name=dyn_leg_r_01, bone2Name=dyn_spine, swing1LimitDegrees=10, swing2LimitDegrees=15, twistLimitDegrees=5}
 {bone1Name=dyn_leg_r_02, bone2Name=dyn_leg_r_01, swing1LimitDegrees=10, swing2LimitDegrees=15, twistLimitDegrees=5}
 {bone1Name=dyn_torso, bone2Name=pack, swing1LimitDegrees=15, swing2LimitDegrees=15, twistLimitDegrees=5}
]
